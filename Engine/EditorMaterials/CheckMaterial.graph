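# CheckMaterial is an unlit, masked editor material for drawing a tintable checkmark.
# It multiplies the GreyCheck texture's RGB by the CheckmarkColor vector parameter, which defaults to green, so the grey image takes on the chosen color.
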Material /Engine/EditorMaterials/CheckMaterial
graph {
 properties {blendMode=BLEND_Masked, opacityMaskClipValue=0.1, pixelDepthOffsetMode=PDOM_Legacy, shadingModel=MSM_Unlit, usedShadingModels=Unlit}
 nodes [
  MaterialExpressionVectorParameter0 = VectorParameter() {parameterName=CheckmarkColor, defaultValue=<rgba 0 1 0 0>}
  MaterialExpressionTextureSample0 = TextureSample() {texture=&/GreyCheck}
  MaterialExpressionMultiply0 = Multiply(A=@MaterialExpressionVectorParameter0.RGB, B=@MaterialExpressionTextureSample0.RGB)
 ]
}
