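# Editor material Cross_Mat is a masked, unlit material built around the Cross texture.
# Its graph multiplies the texture's RGB by a constant mid-grey vector, halving its brightness.
# The file shows no connections from this multiply to the material outputs, so where the result lands is not visible.
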Material /Engine/EditorMaterials/Cross_Mat
graph {
 properties {blendMode=BLEND_Masked, pixelDepthOffsetMode=PDOM_Legacy, shadingModel=MSM_Unlit, usedShadingModels=Unlit}
 nodes [Multiply(A=Constant3Vector() {constant=<rgba 0.5 0.5 0.5 0>}, B=TextureSample() {texture=&/Cross}.RGB)]
}
