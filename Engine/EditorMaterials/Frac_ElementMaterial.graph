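# Translucent, unlit editor material, named for fracture elements in the engine's editor materials.
# Its graph holds a single Vertex Color node, with no connection to any material output shown.
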Material /Engine/EditorMaterials/Frac_ElementMaterial
graph {
 properties {blendMode=BLEND_Translucent, pixelDepthOffsetMode=PDOM_Legacy, shadingModel=MSM_Unlit, usedShadingModels=Unlit}
 nodes [VertexColor()]
}
