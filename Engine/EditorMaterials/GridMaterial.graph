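# GridMaterial is an engine editor material using the Default Lit shading model, presumably for grid display.
# Constant math yields a flat grey diffuse and a zero specular color, both fed into ConvertFromDiffSpec to translate legacy diffuse/specular values.
# A texture sample with tiled UVs feeds nothing in the graph, and no links to material outputs are shown.
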Material /Engine/EditorMaterials/GridMaterial
graph {
 properties {pixelDepthOffsetMode=PDOM_Legacy, usedShadingModels=Default Lit}
 nodes [
  MaterialExpressionConstant1 = Constant() {r=0.25}
  MaterialExpressionMultiply0 = Multiply(A=Constant() {r=1}, B=@MaterialExpressionConstant1)
  MaterialExpressionConstant0 = Constant()
  MaterialFunctionCall(
   DiffuseColor=@MaterialExpressionMultiply0
   SpecularColor=Multiply(A=@MaterialExpressionConstant0) {constA=1, constB=0.04, desc=FresnelBaseReflectFraction}
  ) {materialFunction=&ConvertFromDiffSpec}
  MaterialExpressionTextureCoordinate0 = TextureCoordinate() {uTiling=3, vTiling=3}
  MaterialExpressionTextureSample2 = TextureSample(UVs=@MaterialExpressionTextureCoordinate0)
 ]
}
