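# HelpActorMaterial is an unlit, masked engine editor material that shows an animated help graphic, with usage enabled for particle sprites.
# A FlipBook function steps through a grid of frames in the T_EditorHelp texture, driven by Time scaled by a Speed parameter.
# The frame's inverse is multiplied by green and added back to the frame, pushing the green channel to full while red and blue follow the texture.
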Material /Engine/EditorMaterials/HelpActorMaterial
graph {
 properties {
  bEnableResponsiveAA=true
  bGenerateSphericalParticleNormals=true
  bUsedWithParticleSprites=true
  blendMode=BLEND_Masked
  numCustomizedUVs=1
  pixelDepthOffsetMode=PDOM_Legacy
  shadingModel=MSM_Unlit
  translucencyLightingMode=TLM_Surface
  usedShadingModels=Unlit
 }
 nodes [
  n1 = Constant() {r=4}
  n2 = MaterialFunctionCall(
   "Animation  Phase (0-1)"=Multiply(A=Time(), B=ScalarParameter() {defaultValue=1, parameterName=Speed})
   Number of Rows=@n1
   Number of Columns=@n1
   Texture=TextureObject() {texture=/Engine/EditorResources/T_EditorHelp}
  ) {materialFunction=&FlipBook}
  Add(A=Multiply(A=Constant3Vector() {constant=<rgba 0 1 0 0>}, B=OneMinus(None=@n2.Result)), B=@n2.Result)
  Constant()
  Constant() {r=0.25}
 ]
}
