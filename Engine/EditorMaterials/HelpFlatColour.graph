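# HelpFlatColour is an engine editor material that, going by its name, provides a flat colour.
# It holds a vector parameter named Colour, defaulting to white, so material instances can change the tint.
# Two constant nodes sit alongside it, and the shading model is Default Lit.
# The graph shows no connections between these nodes or to material inputs.
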Material /Engine/EditorMaterials/HelpFlatColour
graph {
 properties {pixelDepthOffsetMode=PDOM_Legacy, usedShadingModels=Default Lit}
 nodes [
  Constant() {r=1}
  Constant() {r=1}
  VectorParameter() {parameterName=Colour, defaultValue=<rgba 1 1 1 1>}
 ]
}
