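# CompositeARGBLayer is an editor material function that composites a Top four-channel color layer over a Bottom one.
# Color is the Bottom RGB scaled by one minus Top alpha, plus the unscaled Top RGB, matching a premultiplied-alpha "over" blend.
# Alpha is one minus the product of both inverted alphas, appended to the color as the single output.
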MaterialFunction /Engine/EditorMaterials/CompositeARGBLayer
graph {nodes [
  n1 = FunctionInput() {inputName=Bottom, inputType=FunctionInput_Vector4}
  n2 = FunctionInput() {inputName=Top, inputType=FunctionInput_Vector4}
  n3 = OneMinus(None=ComponentMask(None=@n2) {a=true})
  n4 = Multiply(A=ComponentMask(None=@n1) {b=true, g=true, r=true}, B=@n3)
  n5 = ComponentMask(None=@n2) {b=true, g=true, r=true}
  n6 = Multiply(A=OneMinus(None=ComponentMask(None=@n1) {a=true}), B=@n3)
  FunctionOutput(None=AppendVector(A=Add(A=@n4, B=@n5), B=OneMinus(None=@n6)))
 ]}
