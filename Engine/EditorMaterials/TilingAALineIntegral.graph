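# Material function in the editor materials that computes the running integral of a tiling 0/1 step mask along x, with Split setting the white-to-black ratio.
# It offsets x by half of Split to center each band on whole numbers, then adds floor(x) times Split to the fractional part capped at Split.
# The name indicates use for analytically antialiased tiling lines; the filtering step itself is outside this function.
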MaterialFunction /Engine/EditorMaterials/TilingAALineIntegral
graph {nodes [
  n1 = FunctionInput() {
   bUsePreviewValueAsDefault=true
   description="0..1, where the mask is white or black"
   inputName=Split
   inputType=FunctionInput_Scalar
   sortPriority=2
   previewValue=<0.1 0 0 1>
  }
  n2 = Add(A=FunctionInput() {inputName=x, inputType=FunctionInput_Scalar}, B=Multiply(A=@n1) {constB=0.5})
  n3 = Floor(None=@n2) {desc="floor(x)"}
  n4 = Subtract(A=@n2, B=@n3) {desc="x-floor(x)"}
  FunctionOutput(None=Add(A=If(A=@n4, B=@n1, "A > B"=@n1, "A < B"=@n4), B=Multiply(A=@n3, B=@n1)))
 ]}
