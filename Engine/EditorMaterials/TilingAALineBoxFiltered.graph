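# Material function that computes an antialiased, box-filtered tiling line mask from a coordinate x, filter half-width h and a Split between white and black.
# It calls TilingAALineIntegral at both ends of the window, x-h and x+h, with the upper end capped at x+1.
# The difference of the two integrals divided by 2h gives the average mask over the window, an analytic filter that needs no supersampling.
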MaterialFunction /Engine/EditorMaterials/TilingAALineBoxFiltered
graph {nodes [
  n1 = FunctionInput() {inputName=x, inputType=FunctionInput_Scalar}
  n2 = FunctionInput() {bUsePreviewValueAsDefault=true, inputName=h, inputType=FunctionInput_Scalar, sortPriority=1, previewValue=<0.01 0 0 1>}
  n3 = FunctionInput() {
   bUsePreviewValueAsDefault=true
   description="0..1, where the mask is white or black"
   inputName=Split
   inputType=FunctionInput_Scalar
   sortPriority=2
   previewValue=<0.1 0 0 1>
  }
  n4 = MaterialFunctionCall!(x=Min(A=Add(A=@n1), B=Add(A=@n2, B=@n1) {desc=x1}) {desc="min(x1+1,x2)"}, Split=@n3) {desc="F(x2)", materialFunction=&/TilingAALineIntegral}
  n5 = MaterialFunctionCall(x=Subtract(A=@n1, B=@n2), Split=@n3) {desc="f(x1)", materialFunction=&TilingAALineIntegral}
  FunctionOutput(None=Divide(A=Subtract(A=@n4.Result, B=@n5.Result), B=Multiply(A=@n2) {constB=2}))
 ]}
