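# Editor material function that outputs a tiling, anti-aliased dot mask from a 2D UV input.
# Each UV axis is padded with a constant and fed to ComputeFilterWidth to get its own filter width.
# Each axis, its width and the shared Split threshold drive TilingAALineBoxFiltered, giving box-filtered lines.
# Multiplying the two line masks yields the dots where they overlap.
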MaterialFunction /Engine/EditorMaterials/TilingAADots
graph {nodes [
  n1 = FunctionInput() {inputName=ScalarUV, inputType=FunctionInput_Vector2}
  n2 = ComponentMask(None=@n1) {r=true}
  n3 = MaterialFunctionCall!(In=AppendVector(A=@n2, B=Constant2Vector())) {materialFunction=/Engine/Functions/Engine_MaterialFunctions03/Procedurals/ComputeFilterWidth}
  n4 = FunctionInput() {
   bUsePreviewValueAsDefault=true
   description="0..1, where the mask is white or black"
   inputName=Split
   inputType=FunctionInput_Scalar
   sortPriority=2
   previewValue=<0.05 0 0 1>
  }
  n5 = ComponentMask(None=@n1) {g=true}
  n6 = MaterialFunctionCall(In=AppendVector(A=@n5, B=Constant2Vector())) {materialFunction=&ComputeFilterWidth}
  FunctionOutput(None=Multiply(
    A=MaterialFunctionCall!(x=@n2, h=@n3.Result, Split=@n4) {materialFunction=&/TilingAALineBoxFiltered}.Result
    B=MaterialFunctionCall(x=@n5, h=@n6.Result, Split=@n4) {materialFunction=&TilingAALineBoxFiltered}.Result
   ))
 ]}
