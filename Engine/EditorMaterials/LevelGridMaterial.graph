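# Unlit, two-sided, masked editor-compositing material that draws the level grid with depth testing disabled.
# TilingAAGrid generates antialiased lines at three UV scales stepped by Exponent, each reduced by AlphaBias for see-through, then merged with screen blends.
# CompositeARGBLayer tints the lines with GridColor and layers TilingAADots snap dots in SnapColor on top.
# A feature-level switch replaces the composite with a constant on ES2 and ES3.1.
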Material /Engine/EditorMaterials/LevelGridMaterial
graph {
 properties {
  bDisableDepthTest=true
  bUsedWithEditorCompositing=true
  blendMode=BLEND_Masked
  opacityMaskClipValue=0.03
  pixelDepthOffsetMode=PDOM_Legacy
  shadingModel=MSM_Unlit
  twoSided=true
  usedShadingModels=Unlit
 }
 nodes [
  n1 = VectorParameter() {parameterName=SnapColor, defaultValue=<rgba 0.5 0 0 0.5>}
  n2 = TextureCoordinate()
  n3 = MaterialFunctionCall!(
   ScalarUV=Multiply(A=@n2, B=ScalarParameter() {defaultValue=50, parameterName=SnapTile}) {constB=10}
   Split=ScalarParameter() {defaultValue=0.075, parameterName=GridSplit}
  ) {materialFunction=&/TilingAADots}
  n4 = ComponentMask(None=VectorParameter() {parameterName=GridColor, defaultValue=<rgba 0.7 0.7 0.9 0.5>}.RGB) {b=true, g=true, r=true}
  n5 = Multiply(A=ScalarParameter() {defaultValue=1, parameterName=GridTile}, B=@n2)
  n6 = Multiply(A=ScalarParameter() {defaultValue=0.05, parameterName=Split})
  n7 = Multiply(A=MaterialFunctionCall!(ScalarUV=@n5, Split=Multiply(A=@n6) {constB=0.25}) {materialFunction=&/TilingAAGrid}.Result)
  n8 = Multiply(A=ScalarParameter() {desc="Bias for more seethrough (0: better quality, more correct but too opaque))", parameterName=AlphaBias})
  n9 = Multiply(A=Clamp(None=Subtract(A=@n7, B=@n8) {constB=0.025, desc=Artificially more translucent so we can see through the grid}))
  n10 = ScalarParameter() {defaultValue=2, parameterName=Exponent}
  n11 = Multiply(A=@n5, B=@n10) {constB=10}
  n12 = Multiply(A=@n6) {constB=0.5}
  n13 = Subtract(A=Multiply(A=MaterialFunctionCall(ScalarUV=@n11, Split=@n12) {materialFunction=&TilingAAGrid}.Result), B=@n8) {constB=0.025, desc=Artificially more translucent so we can see through the grid}
  n14 = Multiply(A=@n11, B=@n10) {constB=10}
  n15 = Multiply(A=@n6)
  n16 = Subtract(A=Multiply(A=MaterialFunctionCall(ScalarUV=@n14, Split=@n15) {materialFunction=&TilingAAGrid}.Result), B=@n8) {constB=0.025, desc=Artificially more translucent so we can see through the grid}
  n17 = MaterialFunctionCall!(Blend=Multiply(A=Clamp(None=@n13)), Base=Multiply(A=Clamp(None=@n16))) {materialFunction=/Engine/Functions/Engine_MaterialFunctions03/Blends/Blend_Screen}
  n18 = Multiply(B=ComponentMask(None=MaterialFunctionCall(Blend=@n9, Base=@n17.Result) {materialFunction=&Blend_Screen}.Result) {r=true}) {constA=1, constB=0.5}
  n19 = MaterialFunctionCall!(
   Top=AppendVector(A=ComponentMask(None=@n1.RGB) {b=true, g=true, r=true}, B=Multiply(A=@n1.A, B=@n3.Result) {constB=0.5})
   Bottom=MaterialFunctionCall(Top=AppendVector(A=@n4, B=@n18), Bottom=Constant4Vector().RGBA) {materialFunction=&/CompositeARGBLayer}.Result
  ) {materialFunction=&CompositeARGBLayer}
  n20 = Constant()
  n21 = FeatureLevelSwitch(Default=@n19.Result, ES2=@n20, ES3_1=@n20, SM4_REMOVED=@n19.Result)
  ComponentMask(None=@n21) {b=true, g=true, r=true}
  ComponentMask(None=@n21) {a=true}
  LinearInterpolate(Alpha=SphereMask(A=WorldPosition().XYZ, B=CameraPositionWS()) {attenuationRadius=10000, hardnessPercent=0}) {constB=0.2}
 ]
}
