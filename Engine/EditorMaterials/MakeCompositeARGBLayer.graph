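# Material function MakeCompositeARGBLayer, an engine editor material helper that builds a premultiplied-alpha color layer for compositing.
# It takes an RGB color and a scalar alpha A, multiplies the color by the alpha, then appends the alpha to output a four-channel value.
# Preview values default to opaque red.
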MaterialFunction /Engine/EditorMaterials/MakeCompositeARGBLayer
graph {nodes [
  n1 = FunctionInput(Preview=Constant3Vector() {constant=<rgba 1 0 0 0>}) {inputName=RGB}
  n2 = FunctionInput(Preview=Constant() {r=1}) {inputName=A, inputType=FunctionInput_Scalar, sortPriority=1}
  FunctionOutput(None=AppendVector(A=Multiply(A=@n1, B=@n2), B=@n2))
 ]}
