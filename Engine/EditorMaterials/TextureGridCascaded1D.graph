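# Editor material function that samples the 1D stripe texture LevelGrid1DStripes, apparently as a building block for level grid materials.
# A scalar ScalarUV input serves as the texture coordinate for a single linear-color sample.
# The sample's RGB output feeds both inputs of an Append node, whose result is the function's only output.
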MaterialFunction /Engine/EditorMaterials/TextureGridCascaded1D
graph {nodes [
  n1 = TextureSample(UVs=FunctionInput() {inputName=ScalarUV, inputType=FunctionInput_Scalar}) {samplerType=SAMPLERTYPE_LinearColor, texture=/Engine/EditorResources/LevelGrid1DStripes}
  FunctionOutput(None=AppendVector(A=@n1.RGB, B=@n1.RGB))
 ]}
