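# Editor material function that extends the cascaded 1D texture grid to two dimensions, producing grid line and point masks from a 2D UV input.
# It splits ScalarUV into R and G channels and runs TextureGridCascaded1D on each axis separately.
# The maximum of the two results gives GridLines, while their product gives Points where lines on both axes intersect.
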MaterialFunction /Engine/EditorMaterials/TextureGridCascaded2D
graph {nodes [
  n1 = FunctionInput() {inputName=ScalarUV, inputType=FunctionInput_Vector2}
  n2 = MaterialFunctionCall!(ScalarUV=ComponentMask(None=@n1) {r=true}) {materialFunction=&/TextureGridCascaded1D}
  n3 = MaterialFunctionCall(ScalarUV=ComponentMask(None=@n1) {g=true}) {materialFunction=&TextureGridCascaded1D}
  FunctionOutput(None=Max(A=@n2.Result, B=@n3.Result)) {outputName=GridLines}
  FunctionOutput(None=Multiply(A=@n2.Result, B=@n3.Result)) {outputName=Points, sortPriority=1}
 ]}
