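# Editor level-grid material: unlit, masked and two-sided, with depth test disabled and editor compositing enabled.
# Scaled texture coordinates, tiled by SnapTile, feed TextureGridCascaded2D, whose larger line channel becomes the grid alpha.
# Soft masks around zero U and V tint the GridColor with UAxisColor and VAxisColor.
# MakeCompositeARGBLayer packs color and alpha, then CompositeARGBLayer stacks it over constant-color layers.
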Material /Engine/EditorMaterials/LevelGridMaterial2
graph {
 properties {
  bDisableDepthTest=true
  bUsedWithEditorCompositing=true
  blendMode=BLEND_Masked
  opacityMaskClipValue=0.03
  pixelDepthOffsetMode=PDOM_Legacy
  shadingModel=MSM_Unlit
  twoSided=true
  usedShadingModels=Unlit
 }
 nodes [
  n1 = VectorParameter() {parameterName=GridColor, defaultValue=<rgba 0.1 0.1 0.12 0.5>}
  n2 = VectorParameter() {parameterName=VAxisColor, defaultValue=<rgba 0 1 0 1>}
  n3 = TextureCoordinate()
  n4 = Constant() {desc=Axis softness, r=350}
  n5 = Abs(None=Multiply(A=ComponentMask(None=@n3) {g=true}, B=@n4))
  n6 = Constant() {desc=Axis thinkness, r=0.25}
  n7 = Clamp(None=OneMinus(None=Subtract(A=@n5, B=@n6)))
  n8 = Multiply(A=ComponentMask(None=@n3) {r=true}, B=@n4) {constB=0.1}
  n9 = OneMinus(None=Subtract(A=Abs(None=@n8), B=@n6))
  n10 = Multiply(A=@n3, B=ScalarParameter() {defaultValue=50, parameterName=SnapTile}) {constB=10}
  n11 = Multiply(A=MaterialFunctionCall!(ScalarUV=Multiply(A=@n10) {constB=0.1}) {materialFunction=&/TextureGridCascaded2D}.GridLines)
  n12 = ComponentMask(None=@n11) {r=true}
  n13 = ComponentMask(None=@n11) {g=true}
  n14 = MaterialFunctionCall!(
   RGB=LinearInterpolate(
    A=LinearInterpolate(A=@n1.RGB, B=@n2.RGB, Alpha=@n7)
    B=VectorParameter() {parameterName=UAxisColor, defaultValue=<rgba 1 0 0 1>}.RGB
    Alpha=Clamp(None=@n9)
   )
   A=Multiply(A=Max(A=@n12, B=@n13)) {constA=1}
  ) {desc=3d grid levels combined, materialFunction=&/MakeCompositeARGBLayer}
  n15 = MaterialFunctionCall!(Top=Constant4Vector().RGBA, Bottom=Constant4Vector() {desc=Bottom Layer Color}.RGBA) {materialFunction=&/CompositeARGBLayer}
  ComponentMask(None=Multiply(A=MaterialFunctionCall(Top=@n14.Result, Bottom=@n15.Result) {materialFunction=&CompositeARGBLayer}.Result)) {a=true}
 ]
}
