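# FakeLight is an editor material function that computes simple lighting from supplied LightPosition and LightColor inputs, independent of scene lights.
# Light and view directions are taken from the object's world position rather than per pixel, then compared with the pixel normal.
# It adds clamped N·L diffuse tinted by LightColor to a tight white Blinn-Phong half-vector highlight.
# The combined result leaves through one output named Diffuse, with no distance attenuation.
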MaterialFunction /Engine/EditorMaterials/FakeLight
graph {
 properties {libraryCategoriesText [Misc,Editor]}
 nodes [
  n1 = FunctionInput() {inputName=LightColor, previewValue=<1 1 1 1>}
  n2 = Normalize(VectorInput=Subtract(A=FunctionInput() {inputName=LightPosition}, B=ObjectPositionWS())) {desc=L}
  n3 = DotProduct(A=PixelNormalWS(), B=@n2)
  n4 = Constant()
  n5 = If(A=@n3, B=@n4, "A > B"=@n3, "A < B"=@n4)
  n6 = Constant3Vector() {constant=<rgba 1 1 1 0>}
  n7 = Normalize(VectorInput=Subtract(A=CameraPositionWS(), B=ObjectPositionWS())) {desc=V}
  n8 = DotProduct(A=Normalize(VectorInput=Add(A=@n2, B=@n7)) {desc=H}, B=PixelNormalWS())
  n9 = Constant()
  n10 = Power(Base=If(A=@n8, B=@n9, "A > B"=@n8, "A < B"=@n9)) {constExponent=120}
  FunctionOutput(None=Add(A=Multiply(A=@n1, B=@n5), B=Multiply(A=@n6, B=@n10))) {outputName=Diffuse}
 ]
}
