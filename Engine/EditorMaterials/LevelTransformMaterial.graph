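# LevelTransformMaterial is a translucent, two-sided editor material with depth testing disabled, drawing a tinted shape over scene geometry.
# A ColorParam vector, translucent green by default, supplies the tint.
# A FakeLight function places a white light at the camera position, and its diffuse result multiplies the tint.
# A dim constant fraction of the tint is added as an ambient floor, so faces turned away from the camera keep some color.
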Material /Engine/EditorMaterials/LevelTransformMaterial
graph {
 properties {bDisableDepthTest=true, blendMode=BLEND_Translucent, pixelDepthOffsetMode=PDOM_Legacy, twoSided=true, usedShadingModels=Default Lit}
 nodes [
  n1 = VectorParameter() {parameterName=ColorParam, defaultValue=<rgba 0 1 0 0.3>}
  n2 = CameraPositionWS() {desc=Light is always at the camera position}
  n3 = Constant3Vector() {constant=<rgba 1 1 1 0>}
  Add(
   A=Multiply(A=@n1.RGB) {constB=0.2}
   B=Multiply(A=MaterialFunctionCall!(LightPosition=@n2, LightColor=@n3) {materialFunction=&/FakeLight}.Diffuse, B=@n1.RGB)
  )
 ]
}
