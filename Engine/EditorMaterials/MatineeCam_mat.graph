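# Editor material for the Matinee camera, judging by its name and location in the engine's EditorMaterials folder.
# A Default Lit surface built from a sample of the MatineeCam_D texture and two scalar constants.
# The graph lists no links from these nodes to material inputs, so their roles are not shown.
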Material /Engine/EditorMaterials/MatineeCam_mat
graph {
 properties {pixelDepthOffsetMode=PDOM_Legacy, usedShadingModels=Default Lit}
 nodes [
  Constant() {r=1}
  Constant() {r=0.2}
  TextureSample() {texture=/Engine/EditorResources/MatineeCam_D}
 ]
}
