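# Editor material function that outputs an antialiased grid mask from a UV input, near 0 on grid lines and 1 inside cells.
# A Custom node finds each axis's distance to the nearest integer line, and Tweak is subtracted to offset the line width.
# Dividing by summed absolute screen-space derivatives keeps edges pixel-sized, with InvSoften scaling their sharpness.
# The clamped X and Y results are multiplied into one mask.
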MaterialFunction /Engine/EditorMaterials/AntialiasedGridFloor
graph {nodes [
  MaterialExpressionFunctionInput_0 = FunctionInput() {inputName=UV, inputType=FunctionInput_Vector2}
  n1 = Custom!(uv=@MaterialExpressionFunctionInput_0) {
   code="min(frac(uv), 1-frac(uv))"
   description="min(frac(uv),1-frac(uv))"
   outputType=CMOT_Float2
   inputs [{inputName=uv, input {expression=&self:MaterialExpressionFunctionInput_0}}]
  }
  n2 = Subtract(A=@n1, B=FunctionInput() {inputName=Tweak, inputType=FunctionInput_Scalar, sortPriority=1, previewValue=<0.005 0 0 1>})
  n3 = Add(A=Abs(None=DDX(Value=@MaterialExpressionFunctionInput_0)), B=Abs(None=DDY(Value=@MaterialExpressionFunctionInput_0))) {desc=ddxy}
  n4 = FunctionInput() {desc="1 for AA lines, smaller for sofer", inputName=InvSoften, inputType=FunctionInput_Scalar, previewValue=<1 0 0 1>}
  n5 = Clamp(None=Multiply(A=Divide(A=@n2, B=@n3), B=@n4) {constB=1.5, desc=soften}) {desc=p new}
  FunctionOutput(None=Multiply(A=ComponentMask(None=@n5) {r=true}, B=ComponentMask(None=@n5) {g=true}))
 ]}
