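# Default Lit editor material for the Persona floor, drawing a tiled grid in muted blue-grey.
# The AntialiasedGridFloor function, replaced by a constant on ES2, combines with a Grid texture to blend two close tints and modulate their darkened brightness.
# Separate branches layer TilePatine diffuse and normal textures at several tilings for color, scalar and normal variation.
# The graph shows no links to material outputs, so which branches are live is unclear.
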Material /Engine/EditorMaterials/PersonaFloorMat
graph {
 properties {bUsedWithStaticLighting=true, pixelDepthOffsetMode=PDOM_Legacy, usedShadingModels=Default Lit}
 nodes [
  n1 = TextureCoordinate() {uTiling=15, vTiling=15}
  n2 = Multiply(A=@n1) {constB=4}
  n3 = TextureCoordinate() {uTiling=30, vTiling=30}
  n4 = Multiply(A=@n3) {constB=4}
  Add(
   A=LinearInterpolate(
    A=TextureSample(UVs=@n2) {samplerType=SAMPLERTYPE_Normal, texture=/Engine/EditorResources/TilePatine_N}.RGB
    B=TextureSample(UVs=@n4) {samplerType=SAMPLERTYPE_Normal, texture=&TilePatine_N}.RGB
   )
   B=Constant3Vector() {constant=<rgba 0 0 2 0>}
  )
  n5 = TextureSample(UVs=TextureCoordinate() {uTiling=400, vTiling=400}) {texture=/Engine/EditorResources/TilePatine_D}
  n6 = TextureSample(UVs=Multiply(A=@n3) {constB=4}) {texture=&TilePatine_D}
  n7 = LinearInterpolate(A=Constant3Vector() {constant=<rgba 0.8 0.18 0.14 0>}, B=LinearInterpolate(A=@n5.RGB, B=@n6.RGB)) {constA=1}
  Clamp(None=Multiply(A=@n7) {constB=3})
  Constant() {r=0.8}
  Constant() {r=1}
  Constant() {r=0.2}
  Desaturation(None=@n7)
  LinearInterpolate(Alpha=TextureSample(UVs=TextureCoordinate() {uTiling=40, vTiling=40}) {texture=&TilePatine_D}.R) {constA=0.25, constB=-0.2}
  LinearInterpolate(A=@n6.R, B=@n5.R) {constAlpha=0.2}
  n8 = Constant3Vector() {constant=<rgba 0.508 0.6227 0.6903 0>}
  n9 = Constant3Vector() {constant=<rgba 0.5927 0.7265 0.8053 0>}
  n10 = MaterialFunctionCall!(InvSoften=Constant() {r=0.75}, UV=@n3, Tweak=Constant() {r=0.005}) {materialFunction=&/AntialiasedGridFloor}
  n11 = FeatureLevelSwitch(Default=@n10.Result, ES2=Constant() {r=0.5}, SM4_REMOVED=@n10.Result)
  n12 = LinearInterpolate(A=OneMinus(None=@n11), B=TextureSample(UVs=@n1) {texture=/Engine/EngineMaterials/Grid}.R, Alpha=@n11)
  Multiply(A=Constant() {r=0.95}, B=Multiply(A=LinearInterpolate(A=@n8, B=@n9, Alpha=@n12), B=LinearInterpolate(Alpha=@n11) {constA=0.6, constB=0.9})) {constB=0.2}
 ]
}
