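# Editor material named for the Persona sky, built around a daylight ambient cubemap.
# The camera vector is negated to give the lookup direction, so the cubemap is sampled along the view ray like a distant environment.
# The cubemap is a texture parameter, so material instances can swap it.
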Material /Engine/EditorMaterials/PersonaSky
graph {
 properties {bUsedWithStaticLighting=true, pixelDepthOffsetMode=PDOM_Legacy, usedShadingModels=Default Lit}
 nodes [MaterialExpressionTextureSampleParameterCube_0 = TextureSampleParameterCube!(UVs=Multiply(A=CameraVectorWS()) {constB=-1}, Apply View MipBias) {parameterName=MaterialExpressionTextureSampleParameterCube_0, texture=/Engine/MapTemplates/Sky/DaylightAmbientCubemap}]
}
