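# Translucent unlit editor material for PhAT (physics asset tool) elements; the file is named Unselected, while the object is named Overlapping.
# It fakes lighting by dotting the pixel world normal with a negated, normalized constant light direction, then bias-scaling and offsetting the result.
# That shade multiplies a muted red color, alongside an Opacity scalar parameter.
# Depth testing is disabled, so elements show through occluding geometry.
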Material /Engine/EditorMaterials/PhAT_ElemUnselectedMaterial.PhAT_ElemOverlappingMaterial
graph {
 properties {
  bDisableDepthTest=true
  blendMode=BLEND_Translucent
  pixelDepthOffsetMode=PDOM_Legacy
  shadingModel=MSM_Unlit
  translucencyPass=MTP_BeforeDOF
  usedShadingModels=Unlit
 }
 nodes [
  MaterialExpressionConstant3Vector4 = Constant3Vector() {constant=<rgba 0.8 0.272 0.272 0>}
  n1 = Constant3Vector() {desc=Fake light direction, constant=<rgba 1 1.5 -3 0>}
  n2 = DotProduct(A=Normalize(VectorInput=Multiply(A=@n1) {constB=-1}), B=PixelNormalWS())
  Multiply(A=@MaterialExpressionConstant3Vector4, B=Add(A=ConstantBiasScale(None=@n2)) {constB=0.1})
  ScalarParameter() {defaultValue=0.6, parameterName=Opacity}
 ]
}
