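# Editor light function material that, by its name, marks preview shadows with a tiled indicator pattern.
# It samples the grayscale PreviewShadowIndicator texture with UVs tiled four times in each direction.
# A feature level switch passes that pattern through by default and substitutes a constant zero on the ES2 and ES3_1 mobile feature levels.
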Material /Engine/EditorMaterials/PreviewShadowIndicatorMaterial
graph {
 properties {materialDomain=MD_LightFunction, pixelDepthOffsetMode=PDOM_Legacy, shadingModel=MSM_Unlit, usedShadingModels=Unlit}
 nodes [
  n1 = TextureSample(UVs=TextureCoordinate() {uTiling=4, vTiling=4}) {samplerType=SAMPLERTYPE_Grayscale, texture=&/PreviewShadowIndicator}
  n2 = Constant()
  FeatureLevelSwitch(Default=@n1.RGB, ES2=@n2, ES3_1=@n2, SM4_REMOVED=@n1.RGB)
 ]
}
