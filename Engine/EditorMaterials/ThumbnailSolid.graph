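# Unlit, two-sided editor material named ThumbnailSolid, apparently a flat solid color for thumbnail rendering.
# Its only node is a vector parameter named Color, with no connection to a material output shown.
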Material /Engine/EditorMaterials/ThumbnailSolid
graph {
 properties {pixelDepthOffsetMode=PDOM_Legacy, shadingModel=MSM_Unlit, twoSided=true, usedShadingModels=Unlit}
 nodes [VectorParameter() {parameterName=Color}]
}
