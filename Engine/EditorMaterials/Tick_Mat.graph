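# Tick_Mat is an unlit, masked editor material that displays the Tick texture in a dimmed tint.
# It multiplies the texture's RGB by a constant mid-light grey, with no parameters or animation.
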Material /Engine/EditorMaterials/Tick_Mat
graph {
 properties {blendMode=BLEND_Masked, pixelDepthOffsetMode=PDOM_Legacy, shadingModel=MSM_Unlit, usedShadingModels=Unlit}
 nodes [Multiply(A=Constant3Vector() {constant=<rgba 0.6 0.6 0.6 0>}, B=TextureSample() {texture=&/Tick}.RGB)]
}
