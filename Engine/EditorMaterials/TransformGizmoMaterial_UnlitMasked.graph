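# Unlit, masked, two-sided editor material for transform gizmos, tinted through the GizmoColor vector parameter.
# Its small graph multiplies the color's RGB by its alpha, with no material output connection shown.
# It disables the depth test, is flagged for use with editor compositing, and casts no ray-traced shadows.
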Material /Engine/EditorMaterials/TransformGizmoMaterial_UnlitMasked
graph {
 properties {
  bCastRayTracedShadows=false
  bDisableDepthTest=true
  bUseTranslucencyVertexFog=false
  bUsedWithEditorCompositing=true
  blendMode=BLEND_Masked
  pixelDepthOffsetMode=PDOM_Legacy
  shadingModel=MSM_Unlit
  twoSided=true
  usedShadingModels=Unlit
 }
 nodes [
  n1 = VectorParameter() {parameterName=GizmoColor, defaultValue=<rgba 1 1 1 1>}
  Multiply(A=@n1.RGB, B=@n1.A)
 ]
}
