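# Editor transform gizmo material: an unlit, masked, two-sided surface with depth test disabled that fakes lighting on vertex colors.
# Camera-space normals are lit by fixed key and fill directions, taking the brighter diffuse term floored by an Ambient parameter.
# That shading scales gamma-adjusted vertex color.
# The result is maxed with vertex-color-tinted Blinn-Phong highlights from both lights, with glossiness set by a parameter.
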Material /Engine/EditorMaterials/TransformGizmoMaterial_VertexColor_PseudoLitMasked
graph {
 properties {
  bCastRayTracedShadows=false
  bDisableDepthTest=true
  bUseTranslucencyVertexFog=false
  bUsedWithEditorCompositing=true
  blendMode=BLEND_Masked
  pixelDepthOffsetMode=PDOM_Legacy
  shadingModel=MSM_Unlit
  twoSided=true
  usedShadingModels=Unlit
 }
 nodes [
  n1 = VertexColor()
  n2 = Transform(None=PixelNormalWS()) {transformSourceType=TRANSFORMSOURCE_World, transformType=TRANSFORM_Camera}
  n3 = Constant() {desc=Almost Zero, r=0.0001}
  n4 = Normalize(VectorInput=Add(A=Constant3Vector() {desc=Key Light Direction, constant=<rgba -1 1 -1 0>}, B=@n3) {constB=0})
  n5 = Normalize(VectorInput=Add(A=Constant3Vector() {desc=Fill Light Direction, constant=<rgba 0 0 -1 0>}, B=@n3) {constB=0})
  n6 = Max(A=DotProduct(A=@n2, B=@n4), B=Multiply(A=DotProduct(A=@n2, B=@n5)) {constB=0.6, desc=Fill Intensity})
  n7 = ScalarParameter() {defaultValue=0.15, desc=Ambient Value, parameterName=Ambient}
  n8 = Transform(None=Constant3Vector() {constant=<rgba 0 0 1 0>}) {transformSourceType=TRANSFORMSOURCE_View, transformType=TRANSFORM_Camera}
  n9 = ScalarParameter() {defaultValue=16, desc=Specular Glossiness, parameterName=SpecularGlossiness}
  n10 = Max(
   A=MaterialFunctionCall!(Normal=@n2, Light Vector=@n4, ViewVector=@n8, Glossiness=@n9) {materialFunction=/Engine/Functions/Engine_MaterialFunctions01/Lighting/BlinnPhongSpecular}.Result
   B=MaterialFunctionCall(Normal=@n2, Light Vector=@n5, ViewVector=@n8, Glossiness=@n9) {materialFunction=&BlinnPhongSpecular}.Result
  )
  Max(
   A=Multiply(A=Power(Base=@n1) {constExponent=2.2}, B=Max(A=@n6, B=@n7) {constB=0.3}) {desc=Add shading}
   B=Multiply(A=@n1, B=Multiply(A=@n10) {constB=0.6, desc=Specular Intensity}) {desc=Colorize Specular}
  )
 ]
}
