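# Editor material for transform gizmos that shades geometry by its vertex color, using an unlit, masked, two-sided setup.
# It raises vertex color to a 2.2 power, a gamma-to-linear style conversion, then multiplies the result by vertex alpha.
# Depth testing is disabled and editor compositing usage is enabled, so gizmos can draw over scene geometry.
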Material /Engine/EditorMaterials/TransformGizmoMaterial_VertexColor_UnlitMasked
graph {
 properties {
  bCastRayTracedShadows=false
  bDisableDepthTest=true
  bUseTranslucencyVertexFog=false
  bUsedWithEditorCompositing=true
  blendMode=BLEND_Masked
  pixelDepthOffsetMode=PDOM_Legacy
  shadingModel=MSM_Unlit
  twoSided=true
  usedShadingModels=Unlit
 }
 nodes [
  n1 = VertexColor()
  Multiply(A=Power(Base=@n1) {constExponent=2.2}, B=@n1.A)
 ]
}
