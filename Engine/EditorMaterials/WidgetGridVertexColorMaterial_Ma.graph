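# An editor material, named for widget grids, that tints a texture with per-vertex color.
# It multiplies the RGB of a texture sample by the vertex color in a single node chain.
# It is unlit, translucent and two-sided, with depth testing disabled so it draws over scene geometry, and it renders in the before-depth-of-field translucency pass.
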Material /Engine/EditorMaterials/WidgetGridVertexColorMaterial_Ma
graph {
 properties {
  bDisableDepthTest=true
  bUseTranslucencyVertexFog=false
  blendMode=BLEND_Translucent
  pixelDepthOffsetMode=PDOM_Legacy
  shadingModel=MSM_Unlit
  translucencyPass=MTP_BeforeDOF
  twoSided=true
  usedShadingModels=Unlit
 }
 nodes [Multiply(A=TextureSample() {texture=&/WidgetGridVertexColorMaterial}.RGB, B=VertexColor())]
}
