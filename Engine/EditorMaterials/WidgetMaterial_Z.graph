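# Editor material for the Z axis of transform widgets, going by its name, flagged for use with editor compositing.
# It is unlit and two-sided, with legacy pixel depth offset mode.
# The graph holds only a constant blue color and a scalar constant, with no connections to material outputs shown.
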Material /Engine/EditorMaterials/WidgetMaterial_Z
graph {
 properties {bUsedWithEditorCompositing=true, pixelDepthOffsetMode=PDOM_Legacy, shadingModel=MSM_Unlit, twoSided=true, usedShadingModels=Unlit}
 nodes [
  MaterialExpressionConstant1 = Constant() {r=0.75}
  MaterialExpressionConstant3Vector0 = Constant3Vector() {constant=<rgba 0.0251 0.207 0.85 0>}
 ]
}
