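# Editor material for widgets, as its name and folder indicate, built around vertex color.
# It is an unlit, translucent, two-sided material with depth testing disabled, so it draws over scene geometry in the before-DOF translucency pass without vertex fog.
# The graph holds a single VertexColor node, with no textures, parameters or math.
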Material /Engine/EditorMaterials/WidgetVertexColorMaterial
graph {
 properties {
  bDisableDepthTest=true
  bUseTranslucencyVertexFog=false
  blendMode=BLEND_Translucent
  opacityMaskClipValue=0
  pixelDepthOffsetMode=PDOM_Legacy
  shadingModel=MSM_Unlit
  translucencyPass=MTP_BeforeDOF
  twoSided=true
  usedShadingModels=Unlit
  lightmassSettings {
   bCastShadowAsMasked=false
   bOverrideCastShadowAsMasked=false
   bOverrideDiffuseBoost=false
   bOverrideEmissiveBoost=false
   bOverrideExportResolutionScale=false
   diffuseBoost=0
   emissiveBoost=0
   exportResolutionScale=0
  }
 }
 nodes [VertexColor()]
}
