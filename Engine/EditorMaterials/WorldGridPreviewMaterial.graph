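# Editor post-process material that previews a world grid and the loading range around the camera, with parameters for cell size, offset, color, range and enable.
# It derives cell borders from world position, highlights cells whose nearest point is within range, fades other lines with distance, and marks the camera position.
# A static switch flattens math to 2D and adds a range ring; finer subdivisions fill uncovered pixels.
# The overlay is divided by eye adaptation, then added to scene color.
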Material /Engine/EditorMaterials/WorldGridPreviewMaterial
graph {
 properties {
  blendableLocation=BL_SceneColorAfterDOF
  blendableOutputAlpha=true
  materialDomain=MD_PostProcess
  pixelDepthOffsetMode=PDOM_Legacy
  refractionCoverageMode=RCM_CoverageIgnored
  shadingModel=MSM_Unlit
  usedShadingModels=Unlit
 }
 nodes [
  n1 = MaterialFunctionCall(Float3=WorldPosition().XYZ) {materialFunction=&BreakOutFloat3Components}
  n2 = StaticSwitch(True=Constant(), False=Constant() {r=1}, Value=StaticBoolParameter() {defaultValue=true, parameterName=Is_Grid_2D})
  n3 = MaterialFunctionCall(Float3=CameraPositionWS()) {materialFunction=&BreakOutFloat3Components}
  n4 = SphereMask(
   A=MaterialFunctionCall(X=@n1.R, Y=@n1.R, Z=Multiply(A=@n2, B=@n1.B)) {materialFunction=&MakeFloat3}.Result
   B=MaterialFunctionCall(X=@n3.R, Y=@n3.R, Z=Multiply(A=@n2, B=@n3.B)) {materialFunction=&MakeFloat3}.Result
  ) {attenuationRadius=100, hardnessPercent=90}
  n5 = Constant() {r=1}
  n6 = If(A=ScalarParameter() {defaultValue=1, parameterName=Grid_Enabled, sliderMax=1}, B=@n5, "A > B"=@n5, "A == B"=@n5, "A < B"=Constant())
  n7 = ScalarParameter() {parameterName=Grid_LoadingRange, sliderMax=102400, sliderMin=1600}
  n8 = ScalarParameter() {defaultValue=3200, parameterName=Grid_CellSize, sliderMax=102400, sliderMin=1600}
  n9 = Subtract(
   A=Subtract(A=WorldPosition().XYZ, B=Multiply(A=@n8) {constB=0.5})
   B=VectorParameter() {parameterName=Grid_Offset, defaultValue=<rgba 0 0 0 1>}.RGB
  )
  n10 = MaterialFunctionCall(Float3=@n9) {materialFunction=&BreakOutFloat3Components}
  n11 = StaticSwitch(True=Constant(), False=Constant() {r=1}, Value=StaticBoolParameter() {defaultValue=true, parameterName=Is_Grid_2D})
  n12 = MaterialFunctionCall(X=@n10.R, Y=@n10.R, Z=Multiply(A=@n11, B=@n10.B)) {materialFunction=&MakeFloat3}
  n13 = ScalarParameter() {defaultValue=3200, parameterName=Grid_CellSize, sliderMax=102400, sliderMin=1600}
  n14 = Subtract(
   A=Subtract(A=CameraPositionWS(), B=Multiply(A=@n13) {constB=0.5})
   B=VectorParameter() {parameterName=Grid_Offset, defaultValue=<rgba 0 0 0 1>}.RGB
  )
  n15 = MaterialFunctionCall(Float3=@n14) {materialFunction=&BreakOutFloat3Components}
  n16 = MaterialFunctionCall(X=@n15.R, Y=@n15.R, Z=Multiply(A=@n11, B=@n15.B)) {materialFunction=&MakeFloat3}
  n17 = Distance(A=@n16.Result, B=@n12.Result)
  n18 = Constant() {r=1}
  n19 = Constant()
  n20 = If(A=@n7, B=@n17, "A > B"=@n18, "A == B"=@n19, "A < B"=@n19) {constB=102400}
  n21 = If(A=Subtract(A=@n7) {constB=100}, B=@n17, "A > B"=@n19, "A == B"=@n18, "A < B"=@n18) {constB=102400}
  n22 = ScalarParameter() {defaultValue=3200, parameterName=Grid_CellSize, sliderMax=102400, sliderMin=1600}
  n23 = Multiply(A=Floor(None=Divide(A=@n12.Result, B=@n22)), B=@n22)
  n24 = Clamp(None=@n16.Result, Min=@n23, Max=Add(A=@n23, B=@n22))
  n25 = Constant() {r=1}
  n26 = If(
   A=ScalarParameter() {parameterName=Grid_LoadingRange, sliderMax=102400, sliderMin=1600}
   B=Distance(A=Subtract(A=@n16.Result, B=@n24), B=Constant())
   "A > B"=@n25
   "A == B"=@n25
   "A < B"=Constant()
  ) {constB=102400}
  n27 = Constant() {r=0.5}
  n28 = ScalarParameter() {defaultValue=3200, parameterName=Grid_CellSize, sliderMax=102400, sliderMin=1600}
  n29 = Subtract(A=Fmod(A=Abs(None=@n9), B=@n28), B=Divide(A=@n28))
  n30 = Add(A=Abs(None=Multiply(A=@n29) {constB=2})) {constB=100}
  n31 = MaterialFunctionCall(Float3=Saturate(None=Subtract(A=@n30, B=@n28) {constB=100})) {materialFunction=&BreakOutFloat3Components}
  n32 = Constant()
  n33 = StaticBoolParameter() {defaultValue=true, parameterName=Is_Grid_2D}
  n34 = Saturate(None=Add(A=Add(A=@n31.R, B=@n31.R), B=StaticSwitch(True=@n32, False=@n31.B, Value=@n33)))
  n35 = Multiply(A=@n26, B=@n34)
  n36 = Saturate(None=Subtract(A=Multiply(A=@n26, B=@n27), B=@n35))
  n37 = Multiply(A=@n7) {constB=5}
  n38 = Multiply(A=Multiply(A=SphereMask(A=@n12.Result, B=@n16.Result, Radius=@n37) {attenuationRadius=100, hardnessPercent=50}), B=OneMinus(None=@n26))
  n39 = Add(
   A=StaticSwitch(True=Multiply(A=@n20, B=@n21) {constB=0.5}, False=Constant(), Value=StaticBoolParameter() {defaultValue=true, parameterName=Is_Grid_2D})
   B=Add(A=Add(A=@n36, B=@n35), B=Multiply(A=@n38, B=@n34))
  )
  n40 = MaterialFunctionCall(Float3=Multiply(A=Multiply(A=@n6, B=@n39), B=VectorParameter() {parameterName=Grid_Color, defaultValue=<rgba 0 20 11 0>}.RGB)) {materialFunction=&BreakOutFloat3Components}
  n41 = Add(
   A=Multiply(A=@n4) {constB=5}
   B=Multiply(
    A=OneMinus(None=@n4)
    B=MaterialFunctionCall(X=@n40.R, Y=@n40.R, Z=@n40.R) {materialFunction=&MakeFloat4}.Result
   )
  )
  n42 = Ceil(None=@n41)
  n43 = ScalarParameter() {defaultValue=1, parameterName=Grid_Enabled, sliderMax=1}
  n44 = Constant() {r=1}
  n45 = Constant()
  n46 = ScalarParameter() {parameterName=Grid_LoadingRange, sliderMax=102400, sliderMin=1600}
  n47 = MaterialFunctionCall(Float3=@n9) {materialFunction=&BreakOutFloat3Components}
  n48 = MaterialFunctionCall(X=@n47.R, Y=@n47.R) {materialFunction=&MakeFloat3}
  n49 = MaterialFunctionCall(Float3=@n14) {materialFunction=&BreakOutFloat3Components}
  n50 = MaterialFunctionCall(X=@n49.R, Y=@n49.R) {materialFunction=&MakeFloat3}
  n51 = Distance(A=@n50.Result, B=@n48.Result)
  n52 = Constant() {r=1}
  n53 = Constant()
  n54 = Subtract(A=@n46) {constB=200}
  n55 = Multiply(
   A=If(A=@n46, B=@n51, "A > B"=@n52, "A == B"=@n53, "A < B"=@n53) {constB=102400}
   B=Max(A=If(A=@n54, B=@n51, "A > B"=@n53, "A == B"=@n52, "A < B"=@n52) {constB=102400}) {constB=0.1}
  ) {constB=0.5}
  n56 = ScalarParameter() {defaultValue=3200, parameterName=Grid_CellSize, sliderMax=102400, sliderMin=1600}
  n57 = Multiply(A=@n56) {constB=0.1}
  n58 = Frac(None=Multiply(A=Divide(A=@n9, B=@n57)))
  n59 = MaterialFunctionCall(Float3=Round(None=Power(Base=@n58) {constExponent=5})) {materialFunction=&BreakOutFloat3Components}
  n60 = Saturate(None=Add(A=Add(A=@n59.R, B=@n59.R), B=Multiply(A=@n59.B) {constB=0}))
  n61 = ScalarParameter() {parameterName=Grid_LoadingRange, sliderMax=102400, sliderMin=1600}
  n62 = Multiply(A=Floor(None=Divide(A=@n48.Result, B=@n56)), B=@n56)
  n63 = Add(A=@n62, B=@n56)
  n64 = Distance(A=Subtract(A=@n50.Result, B=Clamp(None=@n50.Result, Min=@n62, Max=@n63)), B=Constant())
  n65 = Constant() {r=1}
  n66 = Constant()
  n67 = Multiply(A=Multiply(A=Multiply(A=@n60) {constB=2}, B=If(A=@n61, B=@n64, "A > B"=@n65, "A == B"=@n65, "A < B"=@n66) {constB=102400})) {constB=0.2}
  n68 = Multiply(
   A=Multiply(A=If(A=@n43, B=@n44, "A > B"=@n44, "A == B"=@n44, "A < B"=@n45), B=Add(A=@n55, B=@n67))
   B=VectorParameter() {parameterName=Grid_Color, defaultValue=<rgba 0 20 11 0>}.RGB
  )
  n69 = MaterialFunctionCall(Float3=@n68) {materialFunction=&BreakOutFloat3Components}
  n70 = Multiply(
   A=OneMinus(None=Saturate(None=@n42))
   B=MaterialFunctionCall(X=@n69.R, Y=@n69.R, Z=@n69.R) {materialFunction=&MakeFloat4}.Result
  )
  Add(A=Divide(A=Add(A=@n41, B=@n70) {constB=0}, B=EyeAdaptation().EyeAdaptation), B=SceneTexture() {sceneTextureId=PPI_PostProcessInput0}.Color)
  MaterialFunctionCall(X=@n15.R, Y=@n15.R) {materialFunction=&MakeFloat3}
  Multiply(A=@n68) {constB=0.2}
 ]
}
